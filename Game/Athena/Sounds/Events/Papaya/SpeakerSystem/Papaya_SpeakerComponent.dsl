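; Actor component that plays a Papaya speaker's audio source bus, attached to its owner's root component.
; On BeginPlay, cosmetic-capable instances use a match-ready async action, then check the global speakers-enabled flag or bAlwaysPlay before spawning the sound with configured volume and attenuation.
; The cached audio component is stopped on EndPlay or by the Stop event.
; SetSubmixSend switches its send between the SFX submix (default) and the 3D licensed-content submix.
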(event EventBeginPlay
  (Networking|CanExecuteCosmeticEvents
    (:True
      (Class|FortAsyncActionAuthorityMatchReady|MatchReadyAsync)
      ;[1 0.2758 0.3176] Check if Always Play set to TRUE or if the CVAR is set to TRUE
      (if (or (FortGlobals|GetPapayaSpeakersEnabled) (Variables|Default|GetbAlwaysPlay))
        ;[0.9765 1 0.2945] Check if Source Bus Reference is Valid, if not fail to spawn Audio Component
        (bind _sourcebus (Variables|Default|GetSourceBus)
          (:then
            ;[0.2882 1 0.2592] If good to go, spawn sound attached to Owner Actor's Root Component and cache Audio Component Ref
            (bind _returnvalue (Audio|SpawnSoundAttached :Sound _sourcebus :AttachToComponent (Variables|/Script/Engine.Actor|GetRootComponent (Components|GetOwner)) :AttachPointName "None" :LocationType "KeepRelativeOffset" :bStopWhenAttachedToDestroyed false :VolumeMultiplier (Variables|Default|GetVolume) :PitchMultiplier 1.0 :StartTime 0.0 :AttenuationSettings (Variables|Default|GetAttenuationSettings) :bAutoDestroy true))
            (Variables|Default|SetSourceBusAudioComponent _returnvalue))
          (:else))))))

(event EventEndPlay
  (params (EndPlayReason :type EEndPlayReason%20Enum :default "Destroyed"))
  (bind _sourcebusaudiocomponent (Variables|Default|GetSourceBusAudioComponent)
    (:then
      (Audio|Components|Audio|Stop _sourcebusaudiocomponent))
    (:else)))

(event Custom|Stop
  (bind _sourcebusaudiocomponent (Variables|Default|GetSourceBusAudioComponent)
    (:then
      (Audio|Components|Audio|Stop _sourcebusaudiocomponent))
    (:else)))

(event Custom|SetSubmixSend
  (params (Default :type Boolean))
  (bind _sourcebusaudiocomponent (Variables|Default|GetSourceBusAudioComponent))
  (if Default
    (Audio|Components|Audio|SetSubmixSend _sourcebusaudiocomponent "/Game/Sounds/SubMixes/Fortnite_LicensedContent_3D.Fortnite_LicensedContent_3D" 0.0 :!class /Script/Engine.AudioComponent)
    (Audio|Components|Audio|SetSubmixSend _sourcebusaudiocomponent "/Game/Sounds/SubMixes/Fortnite_SFX.Fortnite_SFX" 1.0 :!class /Script/Engine.AudioComponent)
    (else
      (Audio|Components|Audio|SetSubmixSend _sourcebusaudiocomponent "/Game/Sounds/SubMixes/Fortnite_SFX.Fortnite_SFX" 0.0 :!class /Script/Engine.AudioComponent)
      (Audio|Components|Audio|SetSubmixSend _sourcebusaudiocomponent "/Game/Sounds/SubMixes/Fortnite_LicensedContent_3D.Fortnite_LicensedContent_3D" 1.0 :!class /Script/Engine.AudioComponent))))
